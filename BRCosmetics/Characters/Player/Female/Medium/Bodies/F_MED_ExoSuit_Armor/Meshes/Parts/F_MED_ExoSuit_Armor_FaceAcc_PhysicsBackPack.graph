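# Physics asset for the female medium ExoSuit Armor FaceAcc part (PhysicsBackPack), giving physics bodies to head hair, a bun and four flaps.
# Capsule chains for front hair, back hair and flaps hang from small MSTR bodies on the head with tight swing and twist limits.
# Head, neck, spine_05 and clavicles supply body volumes, with extra boxes on the clavicles.
# Angular-free cross constraints tie the first three left and right back-hair links together.
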PhysicsAsset /BRCosmetics/Characters/Player/Female/Medium/Bodies/F_MED_ExoSuit_Armor/Meshes/Parts/F_MED_ExoSuit_Armor_FaceAcc_PhysicsBackPack
bodies {
 clavicle_l [
  {shapeType=Capsule, length=15.5, radius=7.5, shapeName=clavicle_l_capsule, center=<9 -0.5 -2>, rotation=<rot 90 0 0>}
  {shapeType=Box, extentX=18, extentY=2, extentZ=10, shapeName=clavicle_l_box, center=<8 0 7>}
 ]
 clavicle_r [
  {shapeType=Capsule, length=15.5, radius=7.5, shapeName=clavicle_r_capsule, center=<-9.0001 0.5 2>, rotation=<rot 90 0 0>}
  {shapeType=Box, extentX=18, extentY=2, extentZ=10, shapeName=clavicle_r_box, center=<-8 0 -7>}
 ]
 dyn_back_hair_l_01 [{shapeType=Capsule, length=3, radius=2, shapeName=dyn_back_hair_r_01_capsule, center=<2.472 0 0>, rotation=<rot 90 0 0>}]
 dyn_back_hair_l_02 [{shapeType=Capsule, length=3.375, radius=3.2, shapeName=dyn_back_hair_r_02_capsule, center=<3.5861 -0.0001 0.0001>, rotation=<rot 90 0 0>}]
 dyn_back_hair_l_03 [{shapeType=Capsule, length=2.75, radius=3.2, shapeName=dyn_back_hair_r_03_capsule, center=<4.2561 0 0>, rotation=<rot 90 0 0>}]
 dyn_back_hair_l_04 [{shapeType=Capsule, length=1.625, radius=3.2, shapeName=dyn_back_hair_l_04_capsule, center=<3.7624 -0.2196 0>, rotation=<rot 90 0 0>}]
 dyn_back_hair_l_05 [{shapeType=Capsule, length=1.75, radius=3.2, shapeName=dyn_back_hair_l_05_capsule, center=<3.7624 -0.5248 0>, rotation=<rot 90 0 0>}]
 dyn_back_hair_l_MSTR [{shapeType=Capsule, length=0.505, radius=0.505, shapeName=dyn_back_hair_l_MSTR_capsule, rotation=<rot 0 -90 0>}]
 dyn_back_hair_r_01 [{shapeType=Capsule, length=3, radius=2, shapeName=dyn_back_hair_r_01_capsule, center=<2.472 0 0>, rotation=<rot 90 0 0>}]
 dyn_back_hair_r_02 [{shapeType=Capsule, length=3.375, radius=3.2, shapeName=dyn_back_hair_r_02_capsule, center=<3.5861 -0.0001 0.0001>, rotation=<rot 90 0 0>}]
 dyn_back_hair_r_03 [{shapeType=Capsule, length=2.75, radius=3.2, shapeName=dyn_back_hair_r_03_capsule, center=<4.2561 0 0>, rotation=<rot 90 0 0>}]
 dyn_back_hair_r_04 [{shapeType=Capsule, length=1.625, radius=3.2, shapeName=dyn_back_hair_r_04_capsule, center=<3.7624 -0.2196 0>, rotation=<rot 90 0 0>}]
 dyn_back_hair_r_05 [{shapeType=Capsule, length=1.75, radius=3.2, shapeName=dyn_back_hair_r_05_capsule, center=<3.7624 -0.5248 0>, rotation=<rot 90 0 0>}]
 dyn_back_hair_r_MSTR [{shapeType=Capsule, length=0.505, radius=0.505, shapeName=dyn_back_hair_r_MSTR_capsule, rotation=<rot 0 -90 0>}]
 dyn_bun [{shapeType=Capsule, length=1, radius=2, shapeName=dyn_bun_capsule, center=<2 0 0>, rotation=<rot 90 0 0>}]
 dyn_flap_l_bot_01 [{
   shapeType=Capsule
   length=2.9161
   radius=1.1125
   shapeName=dyn_flap_l_bot_01_capsule
   center=<10.205 -0.6724 0.5668>
   rotation=<rot 0 174.6031 -85.6294>
  }]
 dyn_flap_l_bot_MSTR [{shapeType=Capsule, length=0.505, radius=0.505, shapeName=dyn_flap_l_bot_MSTR_capsule, rotation=<rot 0 -90 0>}]
 dyn_flap_l_top_01 [{
   shapeType=Capsule
   length=2.9161
   radius=1.1125
   shapeName=dyn_flap_l_top_01_capsule
   center=<7.8463 -0.6796 0.4622>
   rotation=<rot 0 174.6032 -85.6294>
  }]
 dyn_flap_l_top_MSTR [{shapeType=Capsule, length=0.505, radius=0.505, shapeName=dyn_flap_l_top_MSTR_capsule, rotation=<rot 0 -90 0>}]
 dyn_flap_r_bot_01 [{
   shapeType=Capsule
   length=2.9161
   radius=1.1125
   shapeName=dyn_flap_r_bot_01_capsule
   center=<10.2051 -0.6724 -0.5668>
   rotation=<rot 0 -5.3969 -85.6294>
  }]
 dyn_flap_r_bot_MSTR [{shapeType=Capsule, length=0.505, radius=0.505, shapeName=dyn_flap_r_bot_MSTR_capsule, rotation=<rot 0 -90 0>}]
 dyn_flap_r_top_01 [{
   shapeType=Capsule
   length=2.9161
   radius=1.1127
   shapeName=dyn_flap_r_top_01_capsule
   center=<7.8609 -0.6821 -0.4842>
   rotation=<rot 0 -5.3835 -85.6351>
  }]
 dyn_flap_r_top_MSTR [{shapeType=Capsule, length=0.505, radius=0.505, shapeName=dyn_flap_r_top_MSTR_capsule, rotation=<rot 0 -90 0>}]
 dyn_front_hair_l_01 [{shapeType=Capsule, length=3.375, radius=1.375, shapeName=dyn_front_hair_l_01_capsule, center=<1 0 0>, rotation=<rot 90 0 0>}]
 dyn_front_hair_l_02 [{shapeType=Capsule, length=3.375, radius=1.375, shapeName=dyn_front_hair_l_02_capsule, center=<1 0 0>, rotation=<rot 90 0 0>}]
 dyn_front_hair_l_03 [{shapeType=Capsule, length=3.375, radius=1.375, shapeName=dyn_front_hair_l_03_capsule, center=<1 0 0>, rotation=<rot 90 0 0>}]
 dyn_front_hair_l_04 [{shapeType=Capsule, length=3.375, radius=1.375, shapeName=dyn_front_hair_l_04_capsule, center=<1 0 0>, rotation=<rot 90 0 0>}]
 dyn_front_hair_l_MSTR [{shapeType=Capsule, length=0.505, radius=0.505, shapeName=dyn_front_hair_l_MSTR_capsule, rotation=<rot 0 -90 0>}]
 dyn_front_hair_r_01 [{shapeType=Capsule, length=3.375, radius=1.375, shapeName=dyn_front_hair_r_01_capsule, center=<1 0 0>, rotation=<rot 90 0 0>}]
 dyn_front_hair_r_02 [{shapeType=Capsule, length=3.375, radius=1.375, shapeName=dyn_front_hair_r_02_capsule, center=<1 0 0>, rotation=<rot 90 0 0>}]
 dyn_front_hair_r_03 [{shapeType=Capsule, length=3.375, radius=1.375, shapeName=dyn_front_hair_r_03_capsule, center=<1 0 0>, rotation=<rot 90 0 0>}]
 dyn_front_hair_r_04 [{shapeType=Capsule, length=3.375, radius=1.375, shapeName=dyn_front_hair_r_04_capsule, center=<1 0 0>, rotation=<rot 90 0 0>}]
 dyn_front_hair_r_MSTR [{shapeType=Capsule, length=0.505, radius=0.505, shapeName=dyn_front_hair_r_MSTR_capsule, rotation=<rot 0 -90 0>}]
 head [{shapeType=Capsule, length=6.3287, radius=4.8882, center=<0.8032 2.4459 0>, rotation=<rot 90 -90 -110.0027>}]
 neck_01 [{shapeType=Capsule, length=4, radius=6.625, shapeName=neck_01_capsule, center=<0 -1.604 0>, rotation=<rot 90 0 0>}]
 neck_02 [{shapeType=Capsule, length=3.625, radius=7, shapeName=neck_02_capsule, center=<0 -1.604 0>, rotation=<rot 90 0 0>}]
 spine_05 [{shapeType=Capsule, length=20.375, radius=11.5, shapeName=spine_05_capsule, center=<-1.8937 -0.422 0.0001>}]
}
constraints [
 {
  bone1Name=dyn_flap_l_top_MSTR
  bone2Name=head
  swing1LimitDegrees=45
  swing1Motion=Free
  swing2LimitDegrees=45
  swing2Motion=Free
  twistLimitDegrees=45
  twistMotion=Free
 }
 {
  bone1Name=dyn_flap_l_bot_MSTR
  bone2Name=head
  swing1LimitDegrees=45
  swing1Motion=Free
  swing2LimitDegrees=45
  swing2Motion=Free
  twistLimitDegrees=45
  twistMotion=Free
 }
 {
  bone1Name=dyn_flap_r_top_MSTR
  bone2Name=head
  swing1LimitDegrees=45
  swing1Motion=Free
  swing2LimitDegrees=45
  swing2Motion=Free
  twistLimitDegrees=45
  twistMotion=Free
 }
 {
  bone1Name=dyn_flap_r_bot_MSTR
  bone2Name=head
  swing1LimitDegrees=45
  swing1Motion=Free
  swing2LimitDegrees=45
  swing2Motion=Free
  twistLimitDegrees=45
  twistMotion=Free
 }
 {
  bone1Name=dyn_front_hair_l_MSTR
  bone2Name=head
  swing1LimitDegrees=45
  swing1Motion=Free
  swing2LimitDegrees=45
  swing2Motion=Free
  twistLimitDegrees=45
  twistMotion=Free
 }
 {
  bone1Name=dyn_front_hair_r_MSTR
  bone2Name=head
  swing1LimitDegrees=45
  swing1Motion=Free
  swing2LimitDegrees=45
  swing2Motion=Free
  twistLimitDegrees=45
  twistMotion=Free
 }
 {
  bone1Name=dyn_back_hair_l_MSTR
  bone2Name=head
  swing1LimitDegrees=45
  swing1Motion=Free
  swing2LimitDegrees=45
  swing2Motion=Free
  twistLimitDegrees=45
  twistMotion=Free
 }
 {
  bone1Name=dyn_back_hair_r_MSTR
  bone2Name=head
  swing1LimitDegrees=45
  swing1Motion=Free
  swing2LimitDegrees=45
  swing2Motion=Free
  twistLimitDegrees=45
  twistMotion=Free
 }
 {bone1Name=dyn_flap_l_top_01, bone2Name=dyn_flap_l_top_MSTR, swing1LimitDegrees=25, swing2LimitDegrees=25, twistLimitDegrees=2}
 {bone1Name=dyn_flap_l_bot_01, bone2Name=dyn_flap_l_bot_MSTR, swing1LimitDegrees=25, swing2LimitDegrees=25, twistLimitDegrees=2}
 {bone1Name=dyn_flap_r_top_01, bone2Name=dyn_flap_r_top_MSTR, swing1LimitDegrees=25, swing2LimitDegrees=25, twistLimitDegrees=2}
 {bone1Name=dyn_flap_r_bot_01, bone2Name=dyn_flap_r_bot_MSTR, swing1LimitDegrees=25, swing2LimitDegrees=25, twistLimitDegrees=2}
 {bone1Name=dyn_front_hair_l_01, bone2Name=dyn_front_hair_l_MSTR, swing1LimitDegrees=15, swing2LimitDegrees=5, twistLimitDegrees=2}
 {bone1Name=dyn_front_hair_l_02, bone2Name=dyn_front_hair_l_01, swing1LimitDegrees=15, swing2LimitDegrees=5, twistLimitDegrees=2}
 {bone1Name=dyn_front_hair_l_03, bone2Name=dyn_front_hair_l_02, swing1LimitDegrees=15, swing2LimitDegrees=10, twistLimitDegrees=2}
 {bone1Name=dyn_front_hair_l_04, bone2Name=dyn_front_hair_l_03, swing1LimitDegrees=15, swing2LimitDegrees=10, twistLimitDegrees=2}
 {bone1Name=dyn_front_hair_r_01, bone2Name=dyn_front_hair_r_MSTR, swing1LimitDegrees=15, swing2LimitDegrees=5, twistLimitDegrees=2}
 {bone1Name=dyn_front_hair_r_02, bone2Name=dyn_front_hair_r_01, swing1LimitDegrees=15, swing2LimitDegrees=5, twistLimitDegrees=2}
 {bone1Name=dyn_front_hair_r_03, bone2Name=dyn_front_hair_r_02, swing1LimitDegrees=15, swing2LimitDegrees=10, twistLimitDegrees=2}
 {bone1Name=dyn_front_hair_r_04, bone2Name=dyn_front_hair_r_03, swing1LimitDegrees=15, swing2LimitDegrees=10, twistLimitDegrees=2}
 {bone1Name=dyn_back_hair_l_01, bone2Name=dyn_back_hair_l_MSTR, swing1LimitDegrees=10, swing2LimitDegrees=10, twistLimitDegrees=2}
 {bone1Name=dyn_back_hair_l_02, bone2Name=dyn_back_hair_l_01, swing1LimitDegrees=10, swing2LimitDegrees=10, twistLimitDegrees=2}
 {bone1Name=dyn_back_hair_l_03, bone2Name=dyn_back_hair_l_02, swing1LimitDegrees=10, swing2LimitDegrees=10, twistLimitDegrees=2}
 {bone1Name=dyn_back_hair_l_04, bone2Name=dyn_back_hair_l_03, swing1LimitDegrees=10, swing2LimitDegrees=10, twistLimitDegrees=2}
 {bone1Name=dyn_back_hair_l_05, bone2Name=dyn_back_hair_l_04, swing1LimitDegrees=15, swing2LimitDegrees=10, twistLimitDegrees=2}
 {bone1Name=dyn_back_hair_r_01, bone2Name=dyn_back_hair_r_MSTR, swing1LimitDegrees=10, swing2LimitDegrees=10, twistLimitDegrees=2}
 {bone1Name=dyn_back_hair_r_02, bone2Name=dyn_back_hair_r_01, swing1LimitDegrees=10, swing2LimitDegrees=10, twistLimitDegrees=2}
 {bone1Name=dyn_back_hair_r_03, bone2Name=dyn_back_hair_r_02, swing1LimitDegrees=10, swing2LimitDegrees=10, twistLimitDegrees=2}
 {bone1Name=dyn_back_hair_r_04, bone2Name=dyn_back_hair_r_03, swing1LimitDegrees=10, swing2LimitDegrees=10, twistLimitDegrees=2}
 {bone1Name=dyn_back_hair_r_05, bone2Name=dyn_back_hair_r_04, swing1LimitDegrees=15, swing2LimitDegrees=10, twistLimitDegrees=2}
 {
  bone1Name=head
  bone2Name=neck_02
  swing1LimitDegrees=45
  swing1Motion=Free
  swing2LimitDegrees=45
  swing2Motion=Free
  twistLimitDegrees=45
  twistMotion=Free
 }
 {
  bone1Name=neck_02
  bone2Name=neck_01
  swing1LimitDegrees=45
  swing1Motion=Free
  swing2LimitDegrees=45
  swing2Motion=Free
  twistLimitDegrees=45
  twistMotion=Free
 }
 {
  bone1Name=clavicle_l
  bone2Name=spine_05
  swing1LimitDegrees=45
  swing1Motion=Free
  swing2LimitDegrees=45
  swing2Motion=Free
  twistLimitDegrees=45
  twistMotion=Free
 }
 {
  bone1Name=clavicle_r
  bone2Name=spine_05
  swing1LimitDegrees=45
  swing1Motion=Free
  swing2LimitDegrees=45
  swing2Motion=Free
  twistLimitDegrees=45
  twistMotion=Free
 }
 {
  bone1Name=neck_01
  bone2Name=spine_05
  swing1LimitDegrees=45
  swing1Motion=Free
  swing2LimitDegrees=45
  swing2Motion=Free
  twistLimitDegrees=45
  twistMotion=Free
 }
 {
  bone1Name=dyn_back_hair_l_03
  bone2Name=dyn_back_hair_r_03
  swing1LimitDegrees=45
  swing1Motion=Free
  swing2LimitDegrees=45
  swing2Motion=Free
  twistLimitDegrees=45
  twistMotion=Free
 }
 {
  bone1Name=dyn_back_hair_l_02
  bone2Name=dyn_back_hair_r_02
  swing1LimitDegrees=45
  swing1Motion=Free
  swing2LimitDegrees=45
  swing2Motion=Free
  twistLimitDegrees=45
  twistMotion=Free
 }
 {
  bone1Name=dyn_back_hair_l_01
  bone2Name=dyn_back_hair_r_01
  swing1LimitDegrees=45
  swing1Motion=Free
  swing2LimitDegrees=45
  swing2Motion=Free
  twistLimitDegrees=45
  twistMotion=Free
 }
 {bone1Name=dyn_bun, bone2Name=head, swing1LimitDegrees=45, swing2LimitDegrees=45, twistLimitDegrees=2}
]
