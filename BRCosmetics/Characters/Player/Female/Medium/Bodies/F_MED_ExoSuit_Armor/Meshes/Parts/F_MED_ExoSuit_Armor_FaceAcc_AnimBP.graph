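# Animation Blueprint for the female medium ExoSuit Armor face accessory, adding secondary motion to back hair and flaps atop the parent mesh's copied pose.
# Hair blends between a plain RigidBody chain and, for melee or short backbling, a body-constrained chain using a backpack physics asset.
# AnimDynamics then drives four flaps, followed by LOD-based head scaling and a targeting flap tilt.
# The EventGraph updates gravity, skydive alpha, melee, backbling and LOD state.
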AnimBlueprint /BRCosmetics/Characters/Player/Female/Medium/Bodies/F_MED_ExoSuit_Armor/Meshes/Parts/F_MED_ExoSuit_Armor_FaceAcc_AnimBP
target_skeleton=&/F_MED_ExoSuit_Armor_FaceAcc_Skeleton
graphs [
 AnimGraph {nodes [
   n1 = RigidBody!(
    ComponentPose=Local To Component(LocalPose=Use cached pose 'Init Cache'())
    Alpha=Get Skydive Alpha() -> (Skydive Alpha:double)
    OverrideWorldGravity:Vector=Get Gravity Override()
    ExternalForce
   ) -> (Pose) {
    bClampLinearTranslationLimitToRefPose=true
    bOverrideWorldGravity=true
    overridePhysicsAsset=&/F_MED_ExoSuit_Armor_FaceAcc_Physics
    simulationSpace=BaseBoneSpace
    alphaBoolBlend {blendOption=Linear}
    baseBoneRef {boneName=head}
    componentAppliedLinearAccClamp=<3000 3000 3000>
    componentLinearAccScale=<0.3 0.3 0.3>
    componentLinearVelScale=<0.3 0.3 0.3>
    overrideWorldGravity {z=-980}
   }
   n2 = "Transform (Modify) Bone - Bone: dyn_back_hair_l_MSTR"!(ComponentPose=@n1, bAlphaBoolEnabled:bool=True, Translation="0,0,0", Rotation="0,-10,0", Scale="1,1,1") -> (Pose) {
    alphaInputType=Bool
    bAlphaBoolEnabled=false
    rotationMode=BMM_Additive
    rotationSpace=BCS_BoneSpace
    translationMode=BMM_Additive
    translationSpace=BCS_BoneSpace
    alphaBoolBlend {blendOption=Linear}
    boneToModify {boneName=dyn_back_hair_l_MSTR}
    rotation {yaw=-10}
   }
   n3 = "Transform (Modify) Bone - Bone: dyn_back_hair_r_MSTR"!(ComponentPose=@n2, bAlphaBoolEnabled:bool=True, Translation="0,0,0", Rotation="0,-10,0", Scale="1,1,1") -> (Pose) {
    alphaInputType=Bool
    bAlphaBoolEnabled=false
    rotationMode=BMM_Additive
    rotationSpace=BCS_BoneSpace
    translationMode=BMM_Additive
    translationSpace=BCS_BoneSpace
    alphaBoolBlend {blendOption=Linear}
    boneToModify {boneName=dyn_back_hair_r_MSTR}
    rotation {yaw=-10}
   }
   n4 = Component To Local(ComponentPose=@n3)
   n5 = Get Skydive Alpha()
   n6 = "Constraint - dyn_back_hair_l_01 (P)"(ComponentPose=Local To Component(LocalPose=Use cached pose 'Init Cache'()), Alpha=@n5, ConstraintWeights_0=1) {
    alphaBoolBlend {blendOption=Linear}
    boneToModify {boneName=dyn_back_hair_l_01}
    constraintSetup [{targetBone {boneName=head}}]
    constraintWeights [1]
   }
   n7 = "Constraint - dyn_back_hair_r_01 (P)"(ComponentPose=@n6, Alpha=@n5, ConstraintWeights_0=1) {
    alphaBoolBlend {blendOption=Linear}
    boneToModify {boneName=dyn_back_hair_r_01}
    constraintSetup [{targetBone {boneName=head}}]
    constraintWeights [1]
   }
   n8 = "Constraint - dyn_back_hair_l_02 (PP)"!(ComponentPose=@n7, Alpha=@n5, ConstraintWeights_0=0.65, ConstraintWeights_1:float=0.35) -> (Pose) {
    alphaBoolBlend {blendOption=Linear}
    boneToModify {boneName=dyn_back_hair_l_02}
    constraintSetup [{targetBone {boneName=neck_02}}, {targetBone {boneName=head}}]
    constraintWeights [1,1]
   }
   n9 = "Constraint - dyn_back_hair_r_02 (PP)"!(ComponentPose=@n8, Alpha=@n5, ConstraintWeights_0=0.65, ConstraintWeights_1:float=0.35) -> (Pose) {
    alphaBoolBlend {blendOption=Linear}
    boneToModify {boneName=dyn_back_hair_r_02}
    constraintSetup [{targetBone {boneName=neck_02}}, {targetBone {boneName=head}}]
    constraintWeights [1,1]
   }
   n10 = "Constraint - dyn_back_hair_l_03 (PPP)"!(ComponentPose=@n9, Alpha=@n5, ConstraintWeights_0=0.4, ConstraintWeights_1:float=0.4, ConstraintWeights_2:float=0.2) -> (Pose) {
    alphaBoolBlend {blendOption=Linear}
    boneToModify {boneName=dyn_back_hair_l_03}
    constraintSetup [{targetBone {boneName=neck_02}}, {targetBone {boneName=neck_01}}, {targetBone {boneName=clavicle_l}}]
    constraintWeights [1,1,1]
   }
   n11 = "Constraint - dyn_back_hair_r_03 (PPP)"!(ComponentPose=@n10, Alpha=@n5, ConstraintWeights_0=0.4, ConstraintWeights_1:float=0.4, ConstraintWeights_2:float=0.2) -> (Pose) {
    alphaBoolBlend {blendOption=Linear}
    boneToModify {boneName=dyn_back_hair_r_03}
    constraintSetup [{targetBone {boneName=neck_02}}, {targetBone {boneName=neck_01}}, {targetBone {boneName=clavicle_r}}]
    constraintWeights [1,1,1]
   }
   n12 = "Constraint - dyn_back_hair_l_04 (PPP)"!(ComponentPose=@n11, Alpha=@n5, ConstraintWeights_0=0.3, ConstraintWeights_1:float=0.4, ConstraintWeights_2:float=0.3) -> (Pose) {
    alphaBoolBlend {blendOption=Linear}
    boneToModify {boneName=dyn_back_hair_l_04}
    constraintSetup [{targetBone {boneName=neck_01}}, {targetBone {boneName=spine_05}}, {targetBone {boneName=clavicle_l}}]
    constraintWeights [1,1,1]
   }
   n13 = "Constraint - dyn_back_hair_r_04 (PPP)"!(ComponentPose=@n12, Alpha=@n5, ConstraintWeights_0=0.3, ConstraintWeights_1:float=0.4, ConstraintWeights_2:float=0.3) -> (Pose) {
    alphaBoolBlend {blendOption=Linear}
    boneToModify {boneName=dyn_back_hair_r_04}
    constraintSetup [{targetBone {boneName=neck_01}}, {targetBone {boneName=spine_05}}, {targetBone {boneName=clavicle_r}}]
    constraintWeights [1,1,1]
   }
   n14 = "Constraint - dyn_back_hair_l_05 (PP)"!(ComponentPose=@n13, Alpha=@n5, ConstraintWeights_0=0.5, ConstraintWeights_1:float=0.5) -> (Pose) {
    alphaBoolBlend {blendOption=Linear}
    boneToModify {boneName=dyn_back_hair_l_05}
    constraintSetup [{targetBone {boneName=spine_05}}, {targetBone {boneName=clavicle_l}}]
    constraintWeights [1,1]
   }
   n15 = "Constraint - dyn_back_hair_r_05 (PP)"!(ComponentPose=@n14, Alpha=@n5, ConstraintWeights_0=0.5, ConstraintWeights_1:float=0.5) -> (Pose) {
    alphaBoolBlend {blendOption=Linear}
    boneToModify {boneName=dyn_back_hair_r_05}
    constraintSetup [{targetBone {boneName=spine_05}}, {targetBone {boneName=clavicle_r}}]
    constraintWeights [1,1]
   }
   n16 = RigidBody(ComponentPose=@n15, Alpha=Get Skydive Alpha(), OverrideWorldGravity=Get Gravity Override()) {
    bClampLinearTranslationLimitToRefPose=true
    bOverrideWorldGravity=true
    overridePhysicsAsset=&/F_MED_ExoSuit_Armor_FaceAcc_PhysicsBackPack
    simulationSpace=BaseBoneSpace
    alphaBoolBlend {blendOption=Linear}
    baseBoneRef {boneName=head}
    componentAppliedLinearAccClamp=<3000 3000 3000>
    componentLinearAccScale=<0.3 0.3 0.3>
    componentLinearVelScale=<0.3 0.3 0.3>
    overrideWorldGravity {z=-980}
   }
   n17 = "Transform (Modify) Bone - Bone: dyn_back_hair_l_MSTR"(ComponentPose=@n16, bAlphaBoolEnabled=True, Translation="0,-3,0", Rotation="0,-15,0", Scale="1,1,1") {
    alphaInputType=Bool
    bAlphaBoolEnabled=false
    rotationMode=BMM_Additive
    rotationSpace=BCS_BoneSpace
    translationMode=BMM_Additive
    translationSpace=BCS_BoneSpace
    alphaBoolBlend {blendOption=Linear}
    boneToModify {boneName=dyn_back_hair_l_MSTR}
    rotation {yaw=-15}
    translation {y=-3}
   }
   n18 = "Transform (Modify) Bone - Bone: dyn_back_hair_r_MSTR"(ComponentPose=@n17, bAlphaBoolEnabled=True, Translation="0,-3,0", Rotation="0,-15,0", Scale="1,1,1") {
    alphaInputType=Bool
    bAlphaBoolEnabled=false
    rotationMode=BMM_Additive
    rotationSpace=BCS_BoneSpace
    translationMode=BMM_Additive
    translationSpace=BCS_BoneSpace
    alphaBoolBlend {blendOption=Linear}
    boneToModify {boneName=dyn_back_hair_r_MSTR}
    rotation {yaw=-15}
    translation {y=-3}
   }
   n19 = Component To Local(ComponentPose=@n18)
   n20 = OR Boolean(A=Get Is Melee() -> (Is Melee:bool), B=Get Is Wearing Backbling())
   n21 = Anim Dynamics - Bone: dyn_flap_l_top_01(
    ComponentPose=Local To Component(LocalPose=Blend Poses by bool(BlendPose_0=@n19, BlendPose_1=@n4, bActiveValue=@n20) {blendType=Linear})
    Alpha=1
    ExternalForce="0,0,0"
   ) {
    angularDampingOverride=0.998
    bGravityOverrideInSimSpace=true
    bOverrideAngularDamping=true
    bUseGravityOverride=true
    simulationSpace=RootRelative
    alphaBoolBlend {blendOption=Linear}
    boundBone {boneName=dyn_flap_l_top_01}
    componentAppliedLinearAccClamp=<5000 5000 5000>
    componentLinearAccScale=<0.3 0.3 0.4>
    componentLinearVelScale=<0.3 0.3 0.3>
    gravityOverride {z=-980}
    physicsBodyDefinitions [{
      sphereCollisionRadius=0
      boundBone {boneName=dyn_flap_l_top_01}
      boxExtents=<12 12 5>
      constraintSetup {angularLimitsMax=<2 3 20>, angularLimitsMin=<-2 -35 -20>, angularTarget=<0 0 0>}
      localJointOffset=<4 0 0>
     }]
    relativeSpaceBone {boneName=dyn_flap_l_top_MSTR}
   }
   n22 = Anim Dynamics - Bone: dyn_flap_l_bot_01(ComponentPose=@n21, Alpha=1, ExternalForce="0,0,0") {
    angularDampingOverride=0.998
    bGravityOverrideInSimSpace=true
    bOverrideAngularDamping=true
    bUseGravityOverride=true
    simulationSpace=RootRelative
    alphaBoolBlend {blendOption=Linear}
    boundBone {boneName=dyn_flap_l_bot_01}
    componentAppliedLinearAccClamp=<5000 5000 5000>
    componentLinearAccScale=<0.3 0.3 0.4>
    componentLinearVelScale=<0.3 0.3 0.3>
    gravityOverride {z=-980}
    physicsBodyDefinitions [{
      sphereCollisionRadius=0
      boundBone {boneName=dyn_flap_l_bot_01}
      boxExtents=<15 15 7>
      constraintSetup {angularLimitsMax=<2 3 20>, angularLimitsMin=<-2 -25 -20>, angularTarget=<0 0 0>}
      localJointOffset=<4 0 0>
     }]
    relativeSpaceBone {boneName=dyn_flap_l_bot_MSTR}
   }
   n23 = Anim Dynamics - Bone: dyn_flap_r_top_01(ComponentPose=@n22, Alpha=1, ExternalForce="0,0,0") {
    angularDampingOverride=0.998
    bGravityOverrideInSimSpace=true
    bOverrideAngularDamping=true
    bUseGravityOverride=true
    simulationSpace=RootRelative
    alphaBoolBlend {blendOption=Linear}
    boundBone {boneName=dyn_flap_r_top_01}
    componentAppliedLinearAccClamp=<5000 5000 5000>
    componentLinearAccScale=<0.3 0.3 0.4>
    componentLinearVelScale=<0.3 0.3 0.3>
    gravityOverride {z=-980}
    physicsBodyDefinitions [{
      sphereCollisionRadius=0
      boundBone {boneName=dyn_flap_r_top_01}
      boxExtents=<12 12 5>
      constraintSetup {angularLimitsMax=<2 35 20>, angularLimitsMin=<-2 -3 -20>, angularTarget=<0 0 0>}
      localJointOffset=<4 0 0>
     }]
    relativeSpaceBone {boneName=dyn_flap_r_top_MSTR}
   }
   n24 = Anim Dynamics - Bone: dyn_flap_r_bot_01(ComponentPose=@n23, Alpha=1, ExternalForce="0,0,0") {
    angularDampingOverride=0.998
    bGravityOverrideInSimSpace=true
    bOverrideAngularDamping=true
    bUseGravityOverride=true
    simulationSpace=RootRelative
    alphaBoolBlend {blendOption=Linear}
    boundBone {boneName=dyn_flap_r_bot_01}
    componentAppliedLinearAccClamp=<5000 5000 5000>
    componentLinearAccScale=<0.3 0.3 0.4>
    componentLinearVelScale=<0.3 0.3 0.3>
    gravityOverride {z=-980}
    physicsBodyDefinitions [{
      sphereCollisionRadius=0
      boundBone {boneName=dyn_flap_r_bot_01}
      boxExtents=<15 15 7>
      constraintSetup {angularLimitsMax=<2 25 20>, angularLimitsMin=<-2 -3 -20>, angularTarget=<0 0 0>}
      localJointOffset=<4 0 0>
     }]
    relativeSpaceBone {boneName=dyn_flap_r_bot_MSTR}
   }
   n25 = "Transform (Modify) Bone - Bone: head"(ComponentPose=@n24, Alpha=Get Current LOD() -> (Current LOD:double), Translation="0,0,0", Rotation="0,0,0", Scale="1.03,1.03,1.03") {scaleMode=BMM_Additive, alphaBoolBlend {blendOption=Linear}, boneToModify {boneName=head}, scale=<1.03 1.03 1.03>}
   n26 = "Transform (Modify) Bone - Bone: dyn_flap_r_bot_MSTR"(ComponentPose=@n25, Alpha="To Float (Boolean)"(InBool=Get bIsTargeting()), Translation="0,0,0", Rotation="-45,0,0", Scale="1,1,1") {
    rotationMode=BMM_Additive
    rotationSpace=BCS_BoneSpace
    alphaBoolBlend {blendOption=Linear}
    alphaScaleBiasClamp {bInterpResult=true}
    boneToModify {boneName=dyn_flap_r_bot_MSTR}
    rotation {pitch=-45}
   }
   Output Pose(Result=Component To Local(ComponentPose=@n26)) {layerGroup=None}
   Save cached pose 'Init Cache'(Pose=Copy Pose From Mesh() {bUseAttachedParent=true}) {cacheName=Init Cache}
  ]}
 EventGraph {nodes [
   Event BlueprintUpdateAnimation()
   > n1 = Sequence!(execute) -> (then_0, then_1, then_2:Exec, then_3:Exec)
   > n2 = Gravity Override State Function DEPRECATED(CharacterPartAnimInstance=Self-Reference(), GameTimeInSeconds=GetGameTimeInSeconds(), GravityOverrideParams=Get Gravity Override Params())
   > Set Gravity Override(Gravity Override=@n2)
   n3 = Branch(execute=@n1.then_1, Condition=OR Boolean(A=Get bIsSkydiveDiveMode(), B=Get bIsSkydiveFloating()))
   > Set Skydive Alpha(Skydive Alpha:double=0.5) -> (Output_Get:double)
   > Set Melee Alpha(Melee Alpha:double=0) -> (Output_Get:double)
   > Set Is Melee(Is Melee:bool=false) -> (Output_Get:bool)
   Set Skydive Alpha(execute=@n3.else, Skydive Alpha=1)
   > n4 = Branch(Condition=Get bIsParachuteOpened())
   n5 = Cast To FortPlayerAnimInstance(execute=@n4.else, Object=GetAnimInstance(self=GetOwnerSkeletalMeshComponent()))
   > Set Melee Alpha(Melee Alpha=0.25)
   > Set Is Melee(Is Melee=false)
   > Switch on EFortWeaponCoreAnimation!(execute, Selection:EFortWeaponCoreAnimation Enum=Get WeaponCoreAnim(self=@n5)) -> (
    Melee:Exec
    Pistol:Exec
    Shotgun:Exec
    PaperBlueprint:Exec
    Rifle:Exec
    MeleeOneHand:Exec
    MachinePistol:Exec
    RocketLauncher:Exec
    GrenadeLauncher:Exec
    GoingCommando:Exec
    AssaultRifle:Exec
    TacticalShotgun:Exec
    SniperRifle:Exec
    TrapPlacement:Exec
    ShoulderLauncher:Exec
    AbilityDecoTool:Exec
    Crossbow:Exec
    C4:Exec
    RemoteControl:Exec
    DualWield:Exec
    AR_BullPup:Exec
    AR_ForwardGrip:Exec
    MedPackPaddles:Exec
    SMG_P90:Exec
    AR_DrumGun:Exec
    Consumable_Small:Exec
    Consumable_Large:Exec
    Balloon:Exec
    MountedTurret:Exec
    CreativeTool:Exec
    ExplosiveBow:Exec
    AshtonIndigo:Exec
    AshtonChicago:Exec
    MeleeDualWield:Exec
    Unarmed:Exec
   )
   > Set Melee Alpha(Melee Alpha=1)
   > Set Is Melee(Is Melee=true)
   n6 = CheckForBackbling(self=Self-Reference())
   Set Is Wearing Backbling(execute=@n1.then_2, Is Wearing Backbling=AND Boolean(A=@n6, B=NOT Boolean(A=@n6.bIsBackblingLong)))
   > Set isNotWearingBackbling(isNotWearingBackbling:bool=NOT Boolean(A=@n6)) -> (Output_Get:bool)
   > Set isWearingCape(isWearingCape=@n6.bIsBackblingLong)
   n7 = Get CurrentLOD()
   Set Current LOD(execute=@n1.then_3, Current LOD:double="To Float (Boolean)"(InBool="To Boolean (Integer)"!(InInt:int=@n7) -> (ReturnValue:bool))) -> (Output_Get:double)
  ]}
]
variables [
 Current LOD {type=number}
 Gravity Override {type=Vector}
 Gravity Override Params {
  type=GravityOverrideParamsStruct
  default {
   onGround_Moving {gravityOverride=<0 0 -800>}
   onGround_Standing {gravityOverride=<0 0 -800>}
   skydive_FloatingDown {gravityOverride=<0 -700 200>}
   skydive_Parachuting {gravityOverride=<0 -300 -980>}
  }
 }
 Is Melee {type=boolean}
 Is Wearing Backbling {type=boolean}
 isNotWearingBackbling {type=boolean}
 isWearingCape {type=boolean}
 Melee Alpha {type=number}
 Skydive Alpha {type=number}
]
